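; Blueprint for a geode prop in the Super V Lair environment set, built around a replicated GeodePhase state.
; When GeodePhase replicates, its RepNotify calls the SetupGeodePhase custom event on remote clients only.
; SetupGeodePhase and the construction script are empty, and BeginPlay and OnWorldReady hold only authority switches with nothing attached, so the asset is a scaffold with no working behavior.
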(event EventBeginPlay
  (Networking|SwitchHasAuthority))

(event EventOnWorldReady
  (Networking|SwitchHasAuthority))

(event Custom|SetupGeodePhase)

(fn OnRep_GeodePhase
  (Networking|SwitchHasAuthority
    (:Remote
      (CallFunction|SetupGeodePhase))))

(fn ConstructionScript)
